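# Looping movement foley cue for the SmartHyena emote, layering idle, forward and acceleration sounds driven by the EmoteMoveSpeed parameter.
# A Mixer blends an idle loop, a forward loop with Doppler and a short Enveloper fade-in, and a non-looping accel wave.
# Each layer has its own continuous modulator mapping that parameter to pitch and volume.
# It uses emote music sound class, attenuation and concurrency, and sends to the music analysis submix.
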SoundCue /Game/Athena/Sounds/Emotes/SmartHyena/EmoteFoley_SmartHyena_Fwd_LP_Cue
root=Mixer_2
nodes [
 {id=Doppler_0, children [WavePlayer_1]}
 {
  id=Enveloper_1
  children [Doppler_0]
  properties {pitchCurve {editorCurveData {keys [{time=0}]}}, volumeCurve {editorCurveData {keys [<at 0 0>,<at 0.1 1.0016>]}}}
 }
 {id=Mixer_2, children [ModulatorContinuous_1, ModulatorContinuous_0, ModulatorContinuous_2], properties {inputVolume [1.3,1.2,1]}}
 {
  id=ModulatorContinuous_0
  children [Enveloper_1]
  properties {pitchModulationParams {minOutput=1, parameterName=EmoteMoveSpeed}, volumeModulationParams {minOutput=0.4, parameterName=EmoteMoveSpeed}}
 }
 {
  id=ModulatorContinuous_1
  children [WavePlayer_0]
  properties {
   pitchModulationParams {maxOutput=1.05, minOutput=0.98, parameterName=EmoteMoveSpeed}
   volumeModulationParams {minOutput=1, parameterName=EmoteMoveSpeed}
  }
 }
 {
  id=ModulatorContinuous_2
  children [WavePlayer_2]
  properties {pitchModulationParams {parameterName=EmoteMoveSpeed}, volumeModulationParams {parameterName=EmoteMoveSpeed}}
 }
 {id=WavePlayer_0, properties {bLooping=true, soundWaveAssetPtr=&/IP_Emote_SmartHyena_Idle_LP}}
 {id=WavePlayer_1, properties {bLooping=true, soundWaveAssetPtr=&/IP_Emote_SmartHyena_Fwd_LP}}
 {id=WavePlayer_2, properties {soundWaveAssetPtr=&/IP_Emote_StartHyena_Accel}}
]
properties {
 attenuationSettings=&Fort_Emotes_Athena_Att
 maxDistance=1600
 soundClassObject=&Fort_Music_Emote
 volumeMultiplier=1
 concurrencyOverrides {maxCount=3}
 concurrencySet [/Game/Sounds/Concurrency_Rules/Athena_EmoteMusic_Conc]
 soundSubmixSends [{minSendLevel=0, soundSubmix=&Fortnite_AudioAnalysis_Music}]
}
